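; Creative device Blueprint for a single concert spotlight, tinting its can mesh and light beam from the color component on cosmetic clients.
; Init and reset enable the custom-color flag through Custom Primitive Data.
; Visual updates wait one tick outside the editor to coalesce repeated timeline calls.
; The can takes a curve-mapped color or black, the beam takes the adjusted color or hides, and near-identical color writes are skipped.
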(event EventResetToInitialValues
  (params (StateInstigator :type Controller%20Object%20Reference))
  (|CallParentFunction StateInstigator)
  (Networking|CanExecuteCosmeticEvents
    (:True
      (Rendering|Material|SetCustomPrimitiveDataFloat (Variables|Debug|GetStaticMeshComponent) 27 1.0) ; bUseCustomColor
      (Rendering|Material|SetCustomPrimitiveDataFloat (Variables|Default|GetLightbeam) 27 1.0) ; bUseCustomColor
    )))

(event EventInitializeDevice
  (|CallParentFunction)
  (Networking|CanExecuteCosmeticEvents
    (:True
      (Rendering|Material|SetCustomPrimitiveDataFloat (Variables|Debug|GetStaticMeshComponent) 27 1.0) ; bUseCustomColor
      (Rendering|Material|SetCustomPrimitiveDataFloat (Variables|Default|GetLightbeam) 27 1.0) ; bUseCustomColor
    )))

(event EventUpdateLightVisualsOnClient
  (Networking|CanExecuteCosmeticEvents
    (:True
      ;[1 1 1] Delays are not functional in editor
      (if (not (Editor|InEditor))
        ;[1 1 1] Optimization (wait for next tick) as this may be called multiple times per tick (timelines). Don't want to use a Do-Once as it will only accept the first call, we want to only accept the last call essentially
        (Utilities|FlowControl|Delay 0.0)
        (else
          _))
      (CallFunction|UpdateLightandBeam))))

(fn GetBeamColor
  (params (AdjustedColor :type F%20HSVA%20Structure))
  (return (Class|CreativeColorComponentBlueprintLibrary|RGBAToVec4 (Class|CreativeColorComponentBlueprintLibrary|HSVAToRGBA AdjustedColor))))

(fn GetLightColor
  (params (AdjustedColor :type F%20HSVA%20Structure))
  (bind _hsva (Class|CreativeColorComponent|GetColorbyCurveVector (Variables|Default|GetColorComponent) (Variables|Curves|GetLightColorCurveVector) AdjustedColor))
  (return (Class|CreativeColorComponentBlueprintLibrary|RGBAToVec4 (Class|CreativeColorComponentBlueprintLibrary|HSVAToRGBA _hsva))))

(fn SetStaticMeshCustomColor
  (params (Target :type Primitive%20Component%20Object%20Reference) (NewColor :type Vector%204%20Structure) (ExistingColor :type Vector%204%20Structure%20%28by%20ref%29))
  (if (not (Class|CreativeColorComponentBlueprintLibrary|NearlyEqual(Vector4) NewColor ExistingColor 0.01))
    (Rendering|Material|SetCustomPrimitiveDataVector4 (Variables|GetTarget) 28 (Variables|GetNewColor)) ; CustomColor
    (Variables|SetBy-RefVar (Variables|GetExistingColor) (Variables|GetNewColor))))

(fn UpdateLightAndBeam
  (bind _blightisvisible (LocalVariables|GetbLightIsVisible))
  (bind _output_get (LocalVariables|SetAdjustedColor (Class|CreativeColorComponent|GetAdjustedColor (Variables|Default|GetColorComponent))))
  (bind _bisvisible (CallFunction|IsColorVisible _output_get))
  (LocalVariables|SetbLightIsVisible _bisvisible)
  (Utilities|FlowControl|Sequence
    (:then_0
      (CallFunction|SetStaticMeshCustomColor (Variables|Default|GetSpotlight_Can) (Utilities|Select (0.0 0.0 0.0 1.0) (CallFunction|GetLightColor (LocalVariables|GetAdjustedColor)) (LocalVariables|GetbLightIsVisible)) (Variables|Default|GetPreviousLightColor)))
    (:then_1
      (Development|SetHiddeninGame (Variables|Default|GetLightbeam) (not _blightisvisible) true)
      (if _blightisvisible
        (CallFunction|SetStaticMeshCustomColor (Variables|Default|GetLightbeam) (CallFunction|GetBeamColor (LocalVariables|GetAdjustedColor)) (Variables|Default|GetPreviousBeamColor))))))

(fn ConstructionScript
  (|CallParentFunction))
